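# Unlit additive, two-sided beam material for the Creative Action Trigger device, colored by a parameter.
# Timestamps and transition length from custom primitive data drive MF_Timestamp_InOut_01, which scales a sphere mask along V to fade the beam in and out from one end.
# A cross-width mask varies hardness and exponent along the length, and depth fade softens intersections.
# SplineThicken sets separate base and tip widths for spline-style meshes.
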Material /Game/Creative/Devices/ActionTrigger/Materials/M_ActionTrigger_Beam
graph {
 properties {
  bEnableResponsiveAA=true
  blendMode=BLEND_Additive
  pixelDepthOffsetMode=PDOM_Legacy
  previewMesh=/Game/Effects/Fort_Effects/Mesh/Splines/SM_Straight_Spline_100_HighPoly
  shadingModel=MSM_Unlit
  twoSided=true
  usedShadingModels=Unlit
 }
 nodes [
  ConstantBiasScale(None=Sine(None=Time()))
  n1 = Constant() {r=1}
  n2 = ScalarParameter() {bUseCustomPrimitiveData=true, parameterName=TransitionLength, primitiveDataIndex=2}
  n3 = MaterialFunctionCall(
   TimestampA=ScalarParameter() {bUseCustomPrimitiveData=true, parameterName=TimestampA}
   TransitionLength_In=@n2
   TimestampB=ScalarParameter() {bUseCustomPrimitiveData=true, parameterName=TimestampB, primitiveDataIndex=1}
   TransitionLength_Out=@n2
  ) {materialFunction=&MF_Timestamp_InOut_01}
  n4 = SphereMask(
   A=ComponentMask(None=TextureCoordinate()) {g=true}
   B=@n1
   Radius=Multiply(A=@n1, B=@n3.SmoothStep)
   Hardness=Multiply(A=@n3.SmoothStep, B=ScalarParameter() {defaultValue=0.25, parameterName=FadeHardness})
  )
  n5 = ScalarParameter() {defaultValue=2, parameterName=FadeExponent}
  n6 = VectorParameter() {parameterName=Color, defaultValue=<rgba 1 0.0715 0.0052 0>}
  n7 = TextureCoordinate()
  n8 = Constant() {r=0.5}
  n9 = ScalarParameter() {defaultValue=0.1, parameterName=HardnessBottom}
  n10 = ScalarParameter() {parameterName=HardnessTop}
  n11 = ComponentMask(None=@n7) {g=true}
  n12 = Power(
   Base=SphereMask(A=ComponentMask(None=@n7) {r=true}, B=@n8, Radius=@n8, Hardness=LinearInterpolate(A=@n9, B=@n10, Alpha=@n11))
   Exp=LinearInterpolate(
    A=ScalarParameter() {defaultValue=3, parameterName=MaskExponentBottom}
    B=ScalarParameter() {defaultValue=6, parameterName=MaskExponentTop}
    Alpha=@n11
   )
  )
  DepthFade(
   Opacity=Multiply(A=Power(Base=@n4, Exp=@n5), B=Multiply(A=@n6.RGB, B=@n12))
   FadeDistance=ScalarParameter() {defaultValue=20, parameterName=FadeDistance}
  )
  MaterialFunctionCall(WidthBase=ScalarParameter() {defaultValue=5, parameterName=WidthBase}, WidthTip=ScalarParameter() {defaultValue=80, parameterName=WidthTip}) {materialFunction=&SplineThicken}
 ]
}
