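# Lit material for the Action Trigger device's eye, drawn as a glowing, pixelated dot-matrix display.
# Stepped UVs and a sphere-masked dot grid pixelate an eye texture that jitters with seeking motion and periodic noise glitches.
# A Sighted parameter blends the eye into a pulsing warning icon, with scrolling scanlines added and a low-to-high color ramp applied.
# Sclera shading and a stepped, time-driven blink mask, controlled by IsActive and Blinking, shape the result.
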Material /Game/Creative/Devices/ActionTrigger/Materials/M_ActionTrigger_Eye
graph {
 properties {pixelDepthOffsetMode=PDOM_Legacy, usedShadingModels=Default Lit}
 nodes [
  n1 = ScalarParameter() {defaultValue=30, parameterName=Steps}
  n2 = Multiply(A=TextureCoordinate(), B=@n1)
  n3 = ConstantBiasScale(None=Frac(None=@n2)) {bias=-0.5, scale=2}
  n4 = Constant()
  n5 = SphereMask(
   A=Power(Base=Abs(None=@n3), Exp=ScalarParameter() {defaultValue=16, parameterName=MatrixExponent})
   B=@n4
   Radius=ScalarParameter() {defaultValue=0.3, parameterName=MatrixRadius}
   Hardness=@n4
  )
  LinearInterpolate(
   A=VectorParameter() {parameterName=BaseColorA, defaultValue=<rgba 0.0019 0.0058 0.033 0>}.RGB
   B=VectorParameter() {parameterName=BaseColorB, defaultValue=<rgba 0.0233 0.1561 0.4323 0>}.RGB
   Alpha=@n5
  )
  n6 = Divide(A=Ceil(None=@n2), B=@n1)
  n7 = ComponentMask(None=@n6) {g=true}
  n8 = ScalarParameter() {defaultValue=7.5, parameterName=ScanlineRepeat}
  n9 = Fmod(
   A=Multiply(A=Time() {bOverride_Period=true, period=2}, B=ScalarParameter() {defaultValue=4, parameterName=ScanlineSpeed})
   B=Constant() {r=2}
  )
  n10 = Add(A=Frac(None=Multiply(A=@n7, B=@n8)), B=Multiply(A=Floor(None=@n9)) {constB=0.5})
  n11 = Constant() {r=0.5}
  n12 = Constant()
  n13 = ScalarParameter() {defaultValue=0.07, parameterName=ScanlineOpacity}
  n14 = Time()
  n15 = Saturate(None=Add(A=Sine(None=@n14) {period=1.8}, B=Sine(None=@n14) {period=3.4932}))
  n16 = ComponentMask(None=Round(None=Power(Base=@n15) {constExponent=15})) {r=true}
  n17 = Panner(Coordinate=AppendVector(
    A=Constant()
    B=MaterialFunctionCall(Float2=@n6) {materialFunction=&BreakOutFloat2Components}.G
   )) {bFractionalPart=true, speedX=0.025, speedY=0.0961}
  n18 = Subtract(A=AppendVector(A=TextureSample(UVs=@n17) {samplerType=SAMPLERTYPE_Normal, texture=/Game/Global/Textures/Default/Noise/64x64Noise}.R, B=Constant())) {constB=0.5}
  n19 = Multiply(A=ScalarParameter() {defaultValue=0.04, parameterName=GlitchScale}, B=@n16)
  n20 = ScalarParameter() {defaultValue=-0.045, parameterName=EyeBiasU}
  n21 = ScalarParameter() {defaultValue=0.01, parameterName=EyeBiasV}
  n22 = Time()
  n23 = ScalarParameter() {defaultValue=8, parameterName=SeekingSteps}
  n24 = Divide(A=Floor(None=Multiply(A=@n22, B=@n23)), B=@n23)
  n25 = Multiply(
   A=AppendVector(A=Sine(None=@n24) {period=6.22}, B=Sine(None=@n24) {period=4.1425})
   B=AppendVector(A=ScalarParameter() {defaultValue=0.035, parameterName=SeekingScaleU}, B=ScalarParameter() {defaultValue=0.015, parameterName=SeekingScaleV})
  )
  n26 = ScalarParameter() {defaultValue=1, parameterName=Blinking}
  n27 = Add(A=Add(A=Multiply(A=@n18, B=@n19), B=@n6), B=Add(A=AppendVector(A=@n20, B=@n21), B=Multiply(A=@n25, B=@n26)) {constB=0})
  n28 = TextureSample(UVs=MaterialFunctionCall(UVs=@n27, Texture Scale=ScalarParameter() {defaultValue=0.95, desc=1.4, parameterName=EyeScaleUV}) {materialFunction=&ScaleUVsByCenter}.UVs) {samplerType=SAMPLERTYPE_LinearColor, texture=/Game/Packages/Fortnite_UniqueMaterials_1/Textures/New_eye_Diffuse2}
  n29 = ScalarParameter() {defaultValue=1.8, parameterName=EyeExponent}
  n30 = AppendVector(A=Constant(), B=ScalarParameter() {defaultValue=-0.08, parameterName=SightedBiasV})
  n31 = TextureSample(UVs=MaterialFunctionCall(UVs=Add(A=@n30, B=@n6), Texture Scale=ScalarParameter() {defaultValue=1.15, parameterName=SightedScaleUV}) {materialFunction=&ScaleUVsByCenter}.UVs) {texture=/Game/UI/Foundation/Textures/Banner/Standard/T-Banners-Icons-Warning-L}
  n32 = Time()
  n33 = ScalarParameter() {parameterName=TimeOffset}
  n34 = ConstantBiasScale(None=Cosine(None=Subtract(A=@n32, B=@n33)))
  n35 = Add(
   A=Multiply(A=SphereMask(A=@n10, B=@n11, Radius=@n11, Hardness=@n12), B=LinearInterpolate(A=@n13, Alpha=@n16) {constB=0.2})
   B=LinearInterpolate(A=Power(Base=@n28.G, Exp=@n29), B=Multiply(A=@n31.RGB, B=@n34), Alpha=ScalarParameter() {parameterName=Sighted})
  )
  n36 = VectorParameter() {parameterName=EyeColorLow, defaultValue=<rgba 0.1686 0.4669 3 0>}
  n37 = VectorParameter() {parameterName=EyeColorHigh, defaultValue=<rgba 0.1319 2.6343 4 0>}
  n38 = Multiply(A=Multiply(A=@n35, B=LinearInterpolate(A=@n36.RGB, B=@n37.RGB, Alpha=@n35)), B=@n5)
  n39 = Constant() {r=0.5}
  n40 = SphereMask(A=@n27, B=@n39, Radius=@n39, Hardness=Constant())
  n41 = Constant()
  n42 = ScalarParameter() {defaultValue=0.45, parameterName=ScleraRadius}
  n43 = ScalarParameter() {parameterName=ScleraHardness}
  n44 = LinearInterpolate(
   A=ScalarParameter() {defaultValue=0.35, parameterName=ScleraShading}
   Alpha=OneMinus(None=SphereMask(A=@n40, B=@n41, Radius=@n42, Hardness=@n43))
  )
  n45 = LinearInterpolate(A=@n40, Alpha=ComponentMask(None=Sine(None=@n27) {period=2}) {g=true})
  n46 = Constant() {r=1}
  n47 = ScalarParameter() {defaultValue=1, parameterName=IsActive}
  n48 = Time()
  n49 = Abs(None=Add(A=Sine(None=@n48) {period=11.4844}, B=Sine(None=@n48) {period=5.5385}))
  n50 = ScalarParameter() {defaultValue=2, parameterName=BlinkSteps}
  n51 = Multiply(A=Saturate(None=Multiply(A=@n49) {constB=15}), B=@n50)
  n52 = LinearInterpolate(A=@n47, B=Divide(A=Floor(None=@n51), B=@n50), Alpha=Multiply(A=@n47, B=ScalarParameter() {defaultValue=1, parameterName=Blinking})) {constA=1}
  n53 = Constant()
  Reroute(None=Multiply(A=Multiply(A=@n38, B=@n44), B=SphereMask(A=@n45, B=@n46, Radius=@n52, Hardness=@n53)))
  ScalarParameter() {defaultValue=0.3, parameterName=Metallic}
  ScalarParameter() {defaultValue=0.3, parameterName=Specular}
  ScalarParameter() {defaultValue=0.3, parameterName=Roughness}
 ]
}
